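# Translucent unlit, two-sided Niagara material for the Space Wanderer skydiving contrail, drawing a scrolling galaxy on ribbons, sprites or meshes.
# A panning nebula texture and a double-sampled space texture are warped by noise-driven BumpOffset.
# They blend by space-texture brightness, then pass through hue shift, desaturation and optional particle tint into emissive.
# Opacity combines both layers' red/blue-weighted brightness with a switchable ribbon or edge-band mask, particle alpha and depth fade.
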Material /ChainsawGameplay/Athena/Effects/Particles/Skydiving/NS_Contrail_SpaceWanderer/Materials/M_SpaceWanderer_01
graph {
 properties {
  bCastRayTracedShadows=false
  bUsedWithNiagaraMeshParticles=true
  bUsedWithNiagaraRibbons=true
  bUsedWithNiagaraSprites=true
  blendMode=BLEND_Translucent
  pixelDepthOffsetMode=PDOM_Legacy
  shadingModel=MSM_Unlit
  twoSided=true
  usedShadingModels=Unlit
 }
 nodes [
  n1 = ParticleColor()
  n2 = AppendVector(
   A=ScalarParameter() {group=TxB Controls, parameterName=Tx_B Pan Speed X}
   B=ScalarParameter() {defaultValue=0.5, group=TxB Controls, parameterName=Tx_B Pan Speed Y}
  )
  n3 = TextureCoordinate()
  n4 = StaticSwitchParameter(
   True=MaterialFunctionCall(UVs=@n3, Rotation Center=Constant2Vector() {g=0.5, r=0.5}, "Rotation Angle (0-1)"=Constant() {r=0.25}) {materialFunction=&CustomRotator}.Rotated Values
   False=@n3
  ) {group=Switches, parameterName=Rotate 90}
  n5 = ScalarParameter() {defaultValue=1, parameterName=TxB Tile X, sliderMin=1}
  n6 = ScalarParameter() {defaultValue=0.35, parameterName=TxB Tile Y}
  n7 = Constant2Vector() {g=0.5, r=2}
  n8 = TextureSample(UVs=Panner(Coordinate=Multiply(A=@n4, B=@n7)) {speedY=-0.7}) {texture=&T_NoisePack_01}
  n9 = BumpOffset(Coordinate=Multiply(A=Panner(Coordinate=@n4, Speed=@n2) {speedY=0.5}, B=AppendVector(A=@n5, B=@n6)), Height=Multiply(A=@n8.G) {constB=2})
  n10 = TextureObject() {texture=/Game/Effects/Fort_Effects/Textures/Tile/T_Portal_Space_D}
  n11 = TextureSample(UVs=Add(A=@n9, B=Constant2Vector() {g=0.5}) {constB=-0.5}, Tex=@n10) {texture=&T_Portal_Space_D}
  n12 = TextureSample(UVs=@n9, Tex=@n10) {texture=&T_Portal_Space_D}
  n13 = MaterialFunctionCall(X=Add(A=@n11.R, B=@n12.R), Y=Add(A=@n11.G, B=@n12.G), Z=Add(A=@n11.B, B=@n12.B)) {materialFunction=&MakeFloat3}
  n14 = MaterialFunctionCall(Float3=@n13.Result) {materialFunction=&BreakOutFloat3Components}
  n15 = Add(A=Add(A=@n14.R, B=@n14.R), B=@n14.B)
  n16 = Multiply(A=@n15, B=ScalarParameter() {defaultValue=1, group=Alpha Controls, parameterName=TxB_AlphaMult})
  n17 = Panner(
   Coordinate=@n4
   Speed=AppendVector(
    A=ScalarParameter() {group=TxA Controls, parameterName=Tx_A Pan Speed X}
    B=ScalarParameter() {defaultValue=0.7, group=TxA Controls, parameterName=Tx_A Pan Speed Y}
   )
  ) {speedY=0.7}
  n18 = AppendVector(A=ScalarParameter() {defaultValue=1.5, parameterName=TxA Tile X}, B=ScalarParameter() {defaultValue=0.2, parameterName=TxA Tile Y})
  n19 = TextureSample(UVs=BumpOffset(Coordinate=Multiply(A=@n17, B=@n18), Height=Multiply(A=@n8.G) {constB=5})) {texture=/Game/Characters/Player/Male/Medium/Bodies/M_Med_Soldier_04/Skins/Galaxy/Textures/T_FN_Nebula}
  n20 = MaterialFunctionCall(Float3=@n19.RGB) {materialFunction=&BreakOutFloat3Components}
  n21 = Multiply(A=Add(A=Add(A=@n20.R, B=@n20.R), B=@n20.B), B=ScalarParameter() {defaultValue=1, group=Alpha Controls, parameterName=TxA_AlphaMult})
  n22 = Power(
   Base=Max(A=Saturate(None=@n16), B=Saturate(None=@n21))
   Exp=ScalarParameter() {defaultValue=0.5, group=Alpha Controls, parameterName=TxMax_Power}
  )
  n23 = Saturate(None=LinearInterpolate(A=@n22, Alpha=@n22))
  n24 = TextureCoordinate() {coordinateIndex=1}
  n25 = ComponentMask(None=@n24) {r=true}
  n26 = OneMinus(None=@n25)
  n27 = Multiply(A=@n26, B=@n26)
  n28 = Subtract(A=@n26, B=Multiply(A=@n25, B=ScalarParameter() {defaultValue=4, group=Ribbon, parameterName=RibbonBaseFade}) {constB=4})
  n29 = Multiply(A=Add(A=@n24) {constB=-0.5}) {constB=2}
  n30 = MaterialFunctionCall(Float2=Multiply(A=@n29, B=@n29)) {materialFunction=&BreakOutFloat2Components}
  n31 = Power(Base=OneMinus(None=@n30.G)) {constExponent=4}
  n32 = Multiply(A=OneMinus(None=Saturate(None=@n28)), B=@n31)
  n33 = ScalarParameter() {
   desc="If less than .5 X_EdgeMask.  if equal to .5 XY_Edge Mask.  if greater than .5 Y_EdgeMask"
   group=Alpha Controls
   parameterName="XY Band Mask (Read Desc)"
  }
  n34 = Power(Base=OneMinus(None=@n30.R)) {constExponent=4}
  n35 = StaticSwitchParameter(
   True=Multiply(A=Multiply(A=@n27, B=@n32), B=ScalarParameter() {defaultValue=1, group=Ribbon, parameterName=RibbonMaskMult}) {constB=2}
   False=If(A=Constant() {r=0.5}, B=@n33, "A > B"=@n31, "A == B"=Multiply(A=@n34, B=@n31), "A < B"=@n34)
  ) {group=Switches, parameterName=isRibbon?}
  DepthFade(
   Opacity=Multiply(A=@n1.A, B=Multiply(A=@n23, B=@n35))
   FadeDistance=ScalarParameter() {group=Alpha Controls, parameterName=DepthFade_Distance}
  )
  n36 = ScalarParameter() {parameterName=Hue Shift Percentage}
  n37 = Constant() {r=0.01}
  n38 = Multiply(A=@n19.RGB, B=ScalarParameter() {defaultValue=2, group=TxA Controls, parameterName=TxA Color Mult})
  n39 = Multiply(A=@n13.Result, B=ScalarParameter() {defaultValue=4, group=TxB Controls, parameterName=TxB Color Mult})
  n40 = Multiply(A=Power(Base=@n15) {constExponent=0.35}) {constB=0.5}
  n41 = Desaturation(
   None=MaterialFunctionCall(Hue Shift Percentage=Multiply(A=@n36, B=@n37), Texture=LinearInterpolate(A=@n38, B=@n39, Alpha=@n40)) {materialFunction=&HueShift}.Result
   Fraction=ScalarParameter() {parameterName=Desaturation}
  )
  MaterialFunctionCall(Emissive=LinearInterpolate(A=@n41, B=Multiply(A=@n41, B=@n1.RGB), Alpha=ScalarParameter() {parameterName=UseParticleColor})) {materialFunction=&MF_FX_EmissiveAdjustments}
 ]
}
